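; Component that runs a GPU shallow-water simulation on ping-ponged height and velocity render targets.
; Initialization creates the targets and pass MIDs, picks the height-advection material by terrain mode unless overridden, binds force and terrain textures, and clears to starting depth.
; Each substep draws height advection, velocity advection, divergence height correction and gravity acceleration, then the velocity force target is cleared after the loop.
; A separate function seeds height from a water velocity texture.
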(event Custom|Initialize%20ShallowWaterSim
  (params (Res :type Integer) (Terrain%20Mode :type FluidTerrainModes%20Enum :default "NewEnumerator0") (Forces :type Texture%20Object%20Reference) (Terrain%20Texture :type Texture%20Object%20Reference) (Default%20Water%20Depth :type Float%20%28double-precision%29))
  (Variables|Default|SetRes Res)
  (Variables|Default|SetTerrainMode Terrain%20Mode)
  (CallFunction|CreateRTs)
  (CallFunction|CreateSimMIDs)
  (CallFunction|SetSimMIDParams Forces Terrain%20Texture)
  (CallFunction|ClearRTs Default%20Water%20Depth))

(event Custom|Release%20Shallow%20Water%20Sim%20RTs
  (Variables|Default|SetHeight%20RT%20A "")
  (Variables|Default|SetHeight%20RT%20B "")
  (Variables|Default|SetVelocity%20RT%20A "")
  (Variables|Default|SetVelocity%20RT%20B "")
  (Variables|Default|SetVelocity%20Force%20RT ""))

(fn ClearRTs
  (params (Default%20Water%20Depth :type Float%20%28double-precision%29))
  (bind _returnvalue (Math|Color|MakeColor 0.0 -99999.0))
  (bind _terrainmode (Variables|Default|GetTerrainMode))
  (bind _returnvalue_2 (Utilities|Enum|Equal(Enum) _terrainmode "NewEnumerator0"))
  (bind _linearcolor (Utilities|Struct|MakeLinearColor Default%20Water%20Depth -99999.0))
  (bind _returnvalue_3 (select _returnvalue_2 _linearcolor _returnvalue))
  (Rendering|ClearRenderTarget2D (Variables|Default|GetHeight%20RT%20A) _returnvalue_3)
  (Rendering|ClearRenderTarget2D (Variables|Default|GetHeight%20RT%20B) _returnvalue_3)
  (Rendering|ClearRenderTarget2D (Variables|Default|GetVelocity%20RT%20A))
  (Rendering|ClearRenderTarget2D (Variables|Default|GetVelocity%20RT%20B)))

(fn CreateRTs
  (Utilities|CreateOrUpdateRT (Variables|Default|GetHeight%20RT%20A) (Variables|Default|GetRes) (Variables|Default|GetRes) "RTF_RGBA16f"
    (:Out
      (Utilities|CreateOrUpdateRT (Variables|Default|GetHeight%20RT%20B) (Variables|Default|GetRes) (Variables|Default|GetRes) "RTF_RGBA16f"
        (:Out
          (Utilities|CreateOrUpdateRT (Variables|Default|GetVelocity%20RT%20A) (Variables|Default|GetRes) (Variables|Default|GetRes) "RTF_RG32f"
            (:Out
              (Utilities|CreateOrUpdateRT (Variables|Default|GetVelocity%20RT%20B) (Variables|Default|GetRes) (Variables|Default|GetRes) "RTF_RG32f"
                (:Out
                  (Utilities|CreateOrUpdateRT (Variables|Default|GetVelocity%20Force%20RT) (Variables|Default|GetRes) (Variables|Default|GetRes) "RTF_RG16f"))))))))))

(fn CreateSimMIDs
  (bind _returnvalue (Rendering|Material|CreateDynamicMaterialInstance (select (Utilities|IsValid (Variables|Default|GetManual%20Advect%20Material) :!class /Script/Engine.KismetSystemLibrary) (Variables|Default|GetManual%20Advect%20Material) (Utilities|Select :Index (Variables|Default|GetTerrainMode) :NewEnumerator0 "/Water/FluidSimulation/Materials/Advect/M_Advect_Height_01.M_Advect_Height_01" :NewEnumerator1 "/Water/FluidSimulation/Materials/Advect/M_Advect_Height_TerrainSystem.M_Advect_Height_TerrainSystem" :NewEnumerator2 "/Water/FluidSimulation/Materials/Advect/M_Advect_Height_TerrainSystem.M_Advect_Height_TerrainSystem" :NewEnumerator3 "/Water/FluidSimulation/Materials/Advect/M_Advect_Height_SceneCapture.M_Advect_Height_SceneCapture")) "Advect Height MID" "Transient" :!class /Script/Engine.KismetMaterialLibrary))
  (Variables|MIDs|SetAdvect%20Height%20MID _returnvalue)
  (bind _returnvalue_2 (Rendering|Material|CreateDynamicMaterialInstance (select (Utilities|IsValid (Variables|Default|GetManual%20Advect%20Velocity%20Material) :!class /Script/Engine.KismetSystemLibrary) (Variables|Default|GetManual%20Advect%20Velocity%20Material) "/Water/FluidSimulation/Materials/Advect/M_Advect_Velocity.M_Advect_Velocity") "Advect Velocity MID" "Transient" :!class /Script/Engine.KismetMaterialLibrary))
  (Variables|MIDs|SetAdvect%20Velocity%20MID _returnvalue_2)
  (bind _returnvalue_3 (Rendering|Material|CreateDynamicMaterialInstance "/Water/FluidSimulation/Materials/Advect/M_CalcDivergence.M_CalcDivergence" "Calc Divergence MID" "Transient" :!class /Script/Engine.KismetMaterialLibrary))
  (Variables|MIDs|SetHeight%20Divergence%20Update%20MID _returnvalue_3)
  (bind _returnvalue_4 (Rendering|Material|CreateDynamicMaterialInstance "/Water/FluidSimulation/Materials/Advect/M_GravAcceleration.M_GravAcceleration" "Grav Acceleration MID" "Transient" :!class /Script/Engine.KismetMaterialLibrary))
  (Variables|MIDs|SetGravity%20Acceleration%20MID _returnvalue_4)
  (Variables|MIDs|SetSimMIDS (Utilities|Array|MakeArray (Variables|MIDs|GetAdvect%20Height%20MID) (Variables|MIDs|GetAdvect%20Velocity%20MID) (Variables|MIDs|GetHeight%20Divergence%20Update%20MID) (Variables|MIDs|GetGravity%20Acceleration%20MID))))

(fn PingPong-Height
  (return (Utilities|Select :Index (Math|Integer|%25%28Integer%29 (Variables|Default|GetHeight%20Idx) 2) :Option%200 (Variables|Default|GetHeight%20RT%20A) :Option%201 (Variables|Default|GetHeight%20RT%20B)) (Utilities|Select :Index (Math|Integer|%25%28Integer%29 (+ (Variables|Default|GetHeight%20Idx) 1) 2) :Option%200 (Variables|Default|GetHeight%20RT%20A) :Option%201 (Variables|Default|GetHeight%20RT%20B))))

(fn PingPong-Velocity
  (return (Utilities|Select :Index (Math|Integer|%25%28Integer%29 (Variables|Default|GetVelocity%20Idx) 2) :Option%200 (Variables|Default|GetVelocity%20RT%20A) :Option%201 (Variables|Default|GetVelocity%20RT%20B)) (Utilities|Select :Index (Math|Integer|%25%28Integer%29 (+ (Variables|Default|GetVelocity%20Idx) 1) 2) :Option%200 (Variables|Default|GetVelocity%20RT%20A) :Option%201 (Variables|Default|GetVelocity%20RT%20B))))

(fn SeedWater
  (bind _returnvalue (Rendering|Material|CreateDynamicMaterialInstance "/Water/FluidSimulation/Materials/Seed/M_SeedWater_01.M_SeedWater_01" "Seed Water MID" "Transient" :!class /Script/Engine.KismetMaterialLibrary))
  (Variables|MIDs|SetSeed%20Water%20MID _returnvalue)
  (Rendering|Material|SetTextureParameterValue (Variables|MIDs|GetSeed%20Water%20MID) "WaterVelocityAndHeight" (Variables|Default|GetWaterVelocityTexture))
  (Variables|Default|SetHeight%20Idx 0)
  (Rendering|ClearRenderTarget2D (Variables|Default|GetHeight%20RT%20A))
  (Rendering|DrawMaterialtoRenderTarget (Variables|Default|GetHeight%20RT%20A) (Variables|MIDs|GetSeed%20Water%20MID)))

(fn SetSimMIDParams
  (params (Forces :type Texture%20Object%20Reference) (Terrain :type Texture%20Object%20Reference))
  (Rendering|Material|SetTextureParameterValue (Variables|MIDs|GetAdvect%20Height%20MID) "ForcesRT" Forces)
  (Rendering|Material|SetTextureParameterValue (Variables|MIDs|GetAdvect%20Velocity%20MID) "VelocityForces" (Variables|Default|GetVelocity%20Force%20RT))
  (Utilities|IsValid Terrain
    (:Is%20Valid
      (Rendering|Material|SetTextureParameterValue (Variables|MIDs|GetAdvect%20Height%20MID) "WaterVelocityAndHeight" Terrain)
      (Rendering|Material|SetScalarParameterValue (Variables|MIDs|GetAdvect%20Height%20MID) "TerrainUseWaterDepthOnly" (Math|Conversions|ToFloat(Boolean) (Utilities|Enum|Equal(Enum) (Variables|Default|GetTerrainMode) "NewEnumerator1")) :!class /Script/Engine.MaterialInstanceDynamic))))

(fn SimulationStep
  (bind (_read_rt _write_rt) (Utilities|HeightPingPong))
  (bind (_read_rt_2 _write_rt_2) (Utilities|VelocityPingPong))
  (bind (_read_rt_3 _write_rt_3) (Utilities|VelocityPingPong))
  (bind (_read_rt_4 _write_rt_4) (Utilities|HeightPingPong))
  (bind _advect_height_mid (Variables|MIDs|GetAdvect%20Height%20MID))
  (bind _height_divergence_update_mid (Variables|MIDs|GetHeight%20Divergence%20Update%20MID))
  (bind _gravity_acceleration_mid (Variables|MIDs|GetGravity%20Acceleration%20MID))
  (for _ (range (+ (Math|Integer|Clamp(Integer) (Variables|Default|GetSubSteps) 0 20) 1))
    (Utilities|FlowControl|Sequence
      (:then_0
        ;[1 1 1] Advect Water Height
        (Rendering|Material|SetTextureParameterValue _advect_height_mid "HeightRT" (Utilities|HeightPingPong))
        (Rendering|Material|SetTextureParameterValue _advect_height_mid "VelocityRT" (Utilities|VelocityPingPong))
        (Rendering|ClearRenderTarget2D _write_rt_4)
        (Rendering|DrawMaterialtoRenderTarget _write_rt_4 (Variables|MIDs|GetAdvect%20Height%20MID))
        (Math|Integer|IncrementInt (Variables|Default|GetHeight%20Idx)))
      (:then_2
        ;[1 1 1] Advect Water Velocity
        (Rendering|Material|SetTextureParameterValue (Variables|MIDs|GetAdvect%20Velocity%20MID) "VelocityRT" (Utilities|VelocityPingPong))
        (Rendering|ClearRenderTarget2D _write_rt_2)
        (Rendering|DrawMaterialtoRenderTarget _write_rt_2 (Variables|MIDs|GetAdvect%20Velocity%20MID))
        (Math|Integer|IncrementInt (Variables|Default|GetVelocity%20Idx)))
      (:then_4
        ;[1 1 1] Divergence Height Correction
        (Rendering|Material|SetTextureParameterValue _height_divergence_update_mid "HeightRT" (Utilities|HeightPingPong))
        (Rendering|Material|SetTextureParameterValue _height_divergence_update_mid "VelocityRT" (Utilities|VelocityPingPong))
        (Rendering|ClearRenderTarget2D _write_rt)
        (Rendering|DrawMaterialtoRenderTarget _write_rt (Variables|MIDs|GetHeight%20Divergence%20Update%20MID))
        (Math|Integer|IncrementInt (Variables|Default|GetHeight%20Idx)))
      (:then_6
        ;[1 1 1] Acceleration from Gravity
        (Rendering|Material|SetTextureParameterValue _gravity_acceleration_mid "HeightRT" (Utilities|HeightPingPong))
        (Rendering|Material|SetTextureParameterValue _gravity_acceleration_mid "VelocityRT" (Utilities|VelocityPingPong))
        (Rendering|ClearRenderTarget2D _write_rt_3)
        (Rendering|DrawMaterialtoRenderTarget _write_rt_3 (Variables|MIDs|GetGravity%20Acceleration%20MID))
        (Math|Integer|IncrementInt (Variables|Default|GetVelocity%20Idx)))))
  (Rendering|ClearRenderTarget2D (Variables|Default|GetVelocity%20Force%20RT))
  (return (Utilities|HeightPingPong)))
