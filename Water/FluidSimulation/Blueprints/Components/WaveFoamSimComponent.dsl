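; Wave foam simulation component that generates foam from a water body's Gerstner waves and a shallow water simulation.
; Initialization allocates render targets and material instances, then packs up to 32 waves into a shared spectrum texture, two pixels per wave.
; Each step draws wave crests using shallow water height, then diffuses and fades foam across ping-pong targets.
; A final delta-copy pass writes current and previous foam frames into the shared FoamRT asset.
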(event Custom|Release%20Foam%20Sim%20RTs
  (Variables|Default|SetFoam%20RT%20A "")
  (Variables|Default|SetFoam%20RT%20B "")
  (Variables|Default|SetWave%20RT ""))

(event Custom|InitializeWaveFoamSim
  (params (Res :type Integer) (WorldSize :type Float%20%28double-precision%29) (Fading :type Float%20%28double-precision%29) (TimeBias :type Float%20%28double-precision%29) (WaterBody :type Water%20Body%20Object%20Reference) (ShallowWater :type Shallow%20Water%20Sim%20Component%20Object%20Reference))
  (Variables|Default|SetRes Res)
  (Variables|Default|SetSimulation%20World%20Size WorldSize)
  (Variables|Default|SetFade Fading)
  (Variables|Default|SetTime%20Bias TimeBias)
  (Variables|Default|Setidx 0)
  (CallFunction|AllocateRTs)
  (CallFunction|CreateMIDs)
  (Utilities|IsValid WaterBody
    (:Is%20Valid
      (bind _asgerstner_water_waves (Utilities|Casting|CastToGerstnerWaterWaves (Water|GetWaterWaves (Water|GetWaterBodyComponent WaterBody)))
        (:then
          (bind _output_get (Variables|Default|SetWaveCopy (Variables|/Script/Water.GerstnerWaterWaves|GetGerstnerWaves _asgerstner_water_waves)))
          (CallFunction|RenderWavesToRT _output_get))
        (:CastFailed
          _)))
    (:Is%20Not%20Valid
      _))
  (Variables|Default|SetShallowWater ShallowWater))

(event Custom|Allocate%20RTs
  (Utilities|AllocateRT (Variables|Default|GetFoam%20RT%20A) (Variables|Default|GetRes) (Variables|Default|GetRes) "RTF_RGBA16f"
    (:Out
      (Utilities|AllocateRT (Variables|Default|GetFoam%20RT%20B) (Variables|Default|GetRes) (Variables|Default|GetRes) "RTF_RGBA16f"
        (:Out
          (Utilities|AllocateRT (Variables|Default|GetWave%20RT) (Variables|Default|GetRes) (Variables|Default|GetRes) "RTF_RGBA16f"))))))

(event Custom|Create%20MIDs
  (Rendering|CreateAndAssignMID (Variables|Default|GetDrawWave%20MID) "/Water/FluidSimulation/WaveFoam/DrawWaveCrests.DrawWaveCrests" "WaveMID"
    (:Out
      (Rendering|CreateAndAssignMID (Variables|Default|GetFoamDiffuse%20MID) "/Water/FluidSimulation/WaveFoam/FoamDiffuse.FoamDiffuse" "FoamDiffuseMID"
        (:Out
          (Rendering|CreateAndAssignMID (Variables|Default|GetFoamToneMapMID) "/Water/FluidSimulation/WaveFoam/FoamDeltaCopy.FoamDeltaCopy" "FoamDeltaMID"
            (:Out
              (Rendering|Material|SetScalarParameterValue "/Water/Materials/MPC/MPC_Water.MPC_Water" "WaveCaptureSize" (Variables|Default|GetSimulation%20World%20Size) :!class /Script/Engine.KismetMaterialLibrary)
              (Rendering|Material|SetTextureParameterValue (Variables|Default|GetDrawWave%20MID) "None" ""))))))))

(event Custom|Wave%20Simulation%20Step
  (bind (_read _write) (Utilities|PingPong))
  (bind (_read_2 _write_2) (Utilities|PingPong))
  (bind _drawwave_mid (Variables|Default|GetDrawWave%20MID))
  (bind _foamtonemapmid (Variables|Default|GetFoamToneMapMID))
  (bind _foamdiffuse_mid (Variables|Default|GetFoamDiffuse%20MID))
  (Rendering|Material|SetScalarParameterValue (Variables|Default|GetFoamDiffuse%20MID) "Fade" (* (Variables|Default|GetFade) 0.1) :!class /Script/Engine.MaterialInstanceDynamic)
  (Rendering|Material|SetScalarParameterValue _drawwave_mid "FoamTimeBias" (Variables|Default|GetTime%20Bias) :!class /Script/Engine.MaterialInstanceDynamic)
  (Rendering|Material|SetTextureParameterValue _drawwave_mid "ShallowWaterRT" (Class|ShallowWaterSimComponent|PingPong-Height (Variables|Default|GetShallowWater)))
  ;[0.4225 0.6232 1] Draw Waves First
  (Rendering|DrawMaterialtoRenderTarget (Variables|Default|GetWave%20RT) (Variables|Default|GetDrawWave%20MID))
  (Rendering|Material|SetTextureParameterValue _foamdiffuse_mid "RT" (Utilities|PingPong))
  (Rendering|Material|SetTextureParameterValue _foamdiffuse_mid "WaveRT" (Variables|Default|GetWave%20RT))
  (Rendering|DrawMaterialtoRenderTarget _write_2 (Variables|Default|GetFoamDiffuse%20MID))
  (Rendering|Material|SetTextureParameterValue _foamtonemapmid "CurrentFrame" _write)
  (Rendering|Material|SetTextureParameterValue _foamtonemapmid "PreviousFrame" _read)
  ;[1 1 1] Copy to Content Browser Asset: Temp
  (Rendering|DrawMaterialtoRenderTarget "/Water/FluidSimulation/WaveFoam/FoamRT.FoamRT" (Variables|Default|GetFoamToneMapMID))
  (Math|Integer|IncrementInt (Variables|Default|Getidx)))

(fn RenderWavestoRT
  (params (NewParam :type Array%20of%20Gerstner%20Wave%20Structures))
  (Rendering|ClearRenderTarget2D "/Water/Textures/DefaultWaveSpectrumRT.DefaultWaveSpectrumRT" (Utilities|Struct|MakeLinearColor 1.0 1e-05 0.0 1.0))
  (bind (_canvas _size _context) (Rendering|BeginDrawCanvastoRenderTarget "/Water/Textures/DefaultWaveSpectrumRT.DefaultWaveSpectrumRT"))
  (Utilities|IsValid _canvas
    (:Is%20Valid
      (for _index (range (+ (Math|Integer|Min(Integer) (Utilities|Array|LastIndex NewParam) (- 32 1)) 1))
        (Canvas|DrawBox _canvas (Math|Vector2D|MakeVector2D (+ (* (Math|Conversions|ToFloat(Integer) _index) 2.0) 0.5) 0.5) "(X=0.500000,Y=0.500000)" 1.0 (Math|Conversions|ToLinearColor(Vector) (.Direction (Utilities|Struct|BreakGerstnerWave (Utilities|Array|Get(acopy) NewParam _index)))))
        (Canvas|DrawBox _canvas (+ (Math|Vector2D|MakeVector2D (+ (* (Math|Conversions|ToFloat(Integer) _index) 2.0) 0.5) 0.5) "(X=1.000000,Y=0.000000)") "(X=0.500000,Y=0.500000)" 1.0 (Utilities|Struct|MakeLinearColor (Utilities|Struct|BreakGerstnerWave (Utilities|Array|Get(acopy) NewParam _index)) (.Amplitude (Utilities|Struct|BreakGerstnerWave (Utilities|Array|Get(acopy) NewParam _index))) (.Steepness (Utilities|Struct|BreakGerstnerWave (Utilities|Array|Get(acopy) NewParam _index))))))
      (Rendering|EndDrawCanvastoRenderTarget _context))))
